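# Skeletal-mesh material for the RoyalDusk pickaxe cosmetic, adding animated caustics and a gem pulse to standard surface maps.
# Time-panned caustics, scaled by CausticsProgress and ColorProgress, join the emissive texture.
# GemPulse moves a soft squared distance band, built from pre-skinned position and an offset vector, scaled by EmissionIntensity; a hue-shifted, power-boosted base color supplies a pink texture.
# Attributes pass through character PostFX and accessory override functions.
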Material /BRCosmetics/Weapons/FORT_Melee/Pickaxe_RoyalDsuk/Materials/M_Pickaxe_RoyalDusk
graph {
 properties {
  bUseMaterialAttributes=true
  bUsedWithMeshDeformer=true
  bUsedWithSkeletalMesh=true
  pixelDepthOffsetMode=PDOM_Legacy
  previewMesh=/BRCosmetics/Weapons/FORT_Melee/Pickaxe_RoyalDsuk/Meshes/RoyalDusk_Axe
  usedShadingModels=Default Lit
 }
 nodes [
  n1 = TextureSample() {samplerType=SAMPLERTYPE_Masks, texture=/BRCosmetics/Weapons/FORT_Melee/Pickaxe_RoyalDsuk/Textures/T_Pickaxe_RoyalDusk_S}
  n2 = Time() {period=5}
  n3 = TextureCoordinate()
  n4 = Multiply(A=NamedRerouteUsage()) {constB=0.3}
  n5 = TextureSample(UVs=Add(A=Multiply(A=@n2) {constB=0.1}, B=Add(A=@n3, B=@n4))) {samplerType=SAMPLERTYPE_LinearColor, texture=/BRCosmetics/Weapons/FORT_Melee/Pickaxe_RoyalDsuk/Textures/T_Pickaxe_RoyalDusk_Caustics}
  n6 = ScalarParameter() {parameterName=CausticsProgress}
  n7 = Multiply(A=Add(B=Multiply(A=@n5.R, B=@n6) {constB=4}) {constA=0.5, constB=0}, B=NamedRerouteUsage())
  n8 = ScalarParameter() {parameterName=ColorProgress}
  n9 = ComponentMask(None=NamedRerouteUsage()) {g=true, r=true}
  n10 = Constant()
  n11 = Add(A=PreSkinnedPosition(), B=Constant3Vector() {constant=<rgba 0 0 -12 0>}) {constB=15}
  n12 = Add(A=Multiply(A=AppendVector(A=@n9, B=@n10)) {constB=2.5}, B=Multiply(A=VertexInterpolator(VS=@n11).PS) {constB=0.02})
  n13 = Constant()
  n14 = Add(A=Subtract(A=Distance(A=@n12, B=@n13), B=ScalarParameter() {parameterName=GemPulse})) {constB=1.5}
  n15 = Saturate(None=OneMinus(None=Abs(None=@n14)))
  n16 = Multiply(A=Multiply(A=@n15, B=@n15), B=Multiply(A=NamedRerouteUsage(), B=ScalarParameter() {parameterName=EmissionIntensity}))
  n17 = TextureSample() {texture=/BRCosmetics/Weapons/FORT_Melee/Pickaxe_RoyalDsuk/Textures/T_Pickaxe_RoyalDusk_E}
  n18 = SetMaterialAttributes!(
   MaterialAttributes
   Base Color=NamedRerouteUsage()
   Specular=@n1.R
   Metallic=@n1.R
   Roughness=@n1.R
   Emissive Color=Add(A=Multiply(A=@n7, B=@n8), B=Add(A=@n16, B=@n17.RGB))
   Normal=NamedRerouteUsage()
  ) {attributeSetTypes [
    #1
    #2
    #3
    #4
    #5
    #6
   ]}
  n19 = MaterialFunctionCall(Material=@n18) {materialFunction=&BaseFN_CharacterFunction_PostFX}
  MaterialFunctionCall(In=@n19.Result) {materialFunction=&MF_TechArtOverrides_accessories}
  n20 = MaterialFunctionCall(
   Hue Shift Percentage=Constant() {r=0.15}
   Texture=NamedRerouteDeclaration(None=TextureSample() {texture=/BRCosmetics/Weapons/FORT_Melee/Pickaxe_RoyalDsuk/Textures/T_Pickaxe_RoyalDusk_D}.RGB) {name=BaseColor, nodeColor=<rgba 0.5929 0 1 1>}
  ) {materialFunction=&HueShift}
  n21 = ScalarParameter() {defaultValue=2.1, parameterName=PinkExpoent}
  NamedRerouteDeclaration(None=Multiply(A=Power(Base=@n20.Result, Exp=@n21)) {constB=5}) {name=PinkTexture, nodeColor=<rgba 1 0 0.7263 1>}
  NamedRerouteDeclaration(None=TextureSample() {samplerType=SAMPLERTYPE_Normal, texture=/BRCosmetics/Weapons/FORT_Melee/Pickaxe_RoyalDsuk/Textures/T_Pickaxe_RoyalDusk_N}.RGB) {name=NormalMap, nodeColor=<rgba 0 1 0.7422 1>}
  NamedRerouteDeclaration(None=@n9) {name=Normal RG, nodeColor=<rgba 0.8047 1 0 1>}
  Time() {bOverride_Period=true, period=4}
 ]
}
